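; BP_PopALoon is an interactable balloon-popping minigame board where players score by bursting balloons that fill at random.
; Interaction activates it, and OnRep_bActive starts or ends the round, audio and balloons.
; The authority rerolls among ten slots for an idle balloon on a timer that shortens to a floor.
; Damage events cast for melee impacts, match the hit sphere to index-aligned render actors, burst them via interface and update the scoreboard.
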(event Custom|RollRandoms
  ;[1 1 1] Difficulty Loop
  (if (Variables|Default|GetbActive)
    (Networking|SwitchHasAuthority
      (:Authority
        (bind _output_get (Variables|Default|SetTempIndex (Math|Random|RandomIntegerinRange 0 9)))
        (if (Utilities|Array|Get(aref) (Variables|Default|GetLivingBalloons) _output_get)
          (Utilities|Time|SetTimerbyFunctionName 0 "RollRandoms" 0.05)
          (else
            (bind _output_get_2 (Variables|Default|SetBalloonIndex _output_get))
            (Utilities|Time|SetTimerbyFunctionName 0 "RollRandoms" (Variables|Default|GetStartingGap))
            (Variables|Default|SetStartingGap (Utilities|ClampMinFloat (- (Variables|Default|GetStartingGap) 0.3) 0.5))
            (CallFunction|TriggerBalloon _output_get_2)))))))

(event Custom|TriggerBalloon
  (params (BalloonToActivate :type Integer))
  ;[1 1 1] Balloon Filling Trigger
  (CallFunction|BalloonFilled BalloonToActivate)
  (Networking|SwitchHasAuthority
    (:Authority
      (Utilities|Array|SetArrayElem (Variables|Default|GetLivingBalloons) BalloonToActivate "true"))
    (:Remote
      _))
  (Class|BalloonBoardInterface|FillBalloon(Message) (Utilities|Array|Get(acopy) (Variables|Default|GetBalloonRender) BalloonToActivate) (Variables|Default|GetFillTimer)))

(event Interaction|EventBlueprintOnInteract
  (params (InteractingPawn :type Fort%20Pawn%20Object%20Reference) (InteractionBeingAttempted :type EInteractionBeingAttempted%20Enum :default "FirstInteraction"))
  ;[0.04501 0.1922 1] What to do when player uses me
  (Class|BPScoreBoardPlace|Startup (Variables|Interaction|GetMyScoreBoard) InteractingPawn)
  (Variables|Default|SetbActive true)
  (CallFunction|ForceTimerReset))

(event EventBeginPlay
  ;[1 1 1] Carried over from Clay Pidgeon
  ;;[1 1 1] Hack delay
  (Utilities|FlowControl|Delay 5.0)
  (Variables|Default|SetStartingGap (Variables|Default|GetSpawnGap))
  ;; Interaction Logic
  (EventDispatchers|AssignOnDamaged (EventDispatchers|CreateEvent :!event Custom|HitCheck))
  (bind _asfort_game_state_athena (Utilities|Casting|CastToFortGameStateAthena (Game|GetGameState)))
  (if (Variables|/Script/FortniteGame.FortGameStateAthena|GetbPlaylistDataIsLoaded _asfort_game_state_athena)
    (if (GameplayTags|HasTag (Class|FortGameStateAthena|GetAthenaPlaylistContextTags _asfort_game_state_athena))
      (Variables|Default|SetCreativeMode true)
      (else
        (Variables|Default|SetCreativeMode false)))
    (Networking|SwitchHasAuthority)
    (else
      (Playlist|AssignOnCurrentPlaylistLoaded _asfort_game_state_athena (delegate Custom|PlaylistLoaded)))))

(event Custom|GameStartup
  ;[0.04501 0.1922 1] Clear board for startup
  (for (_ _array_index) (Variables|Default|GetLivingBalloons)
    (Utilities|Array|SetArrayElem (Variables|Default|GetLivingBalloons) _array_index))
  (bind _returnvalue (Audio|SpawnSoundAttached "/Figment_S07/Sounds/Fort_Gadgets/Balloons/Balloon_Quest_Start_Cue.Balloon_Quest_Start_Cue" (Variables|Components|GetBoardRoot)))
  (Variables|Default|SetGame_Music _returnvalue))

(event Custom|PlaylistLoaded
  (params (PlaylistName :type Name) (PlaylistContextTags :type Gameplay%20Tag%20Container%20Structure%20%28by%20ref%29))
  ;[1 1 1] Carried over from Clay Pidgeon
  (if (GameplayTags|HasTag PlaylistContextTags)
    (Variables|Default|SetCreativeMode true)
    (else
      (Variables|Default|SetCreativeMode false)))
  (Networking|SwitchHasAuthority))

(event Custom|BalloonFilled
  (params (Index :type Integer))
  ;[1 1 1] Active Loop
  (if (Variables|Default|GetbActive)
    (Collision|SetCollisionEnabled (Utilities|Array|Get(acopy) (Variables|Default|GetBalloonCollision) Index) "QueryAndPhysics") ; Turn on collision
    (CallFunction|SetFilling Index)
    (else
      ;;[1 1 1] Board isn't active, tell them all to die
      (Class|BalloonBoardInterface|KillBalloons(Message) (Utilities|Array|Get(acopy) (Variables|Default|GetBalloonRender) Index))
      (Collision|SetCollisionEnabled (Utilities|Array|Get(acopy) (Variables|Default|GetBalloonCollision) Index) "NoCollision"))))

(event Custom|HitCheck
  (params (DamagedActor :type Actor%20Object%20Reference) (Damage :type Float%20%28single-precision%29) (InstigatedBy :type Controller%20Object%20Reference) (DamageCauser :type Actor%20Object%20Reference) (HitLocation :type Vector) (FHitComponent :type Primitive%20Component%20Object%20Reference) (BoneName :type Name) (Momentum :type Vector))
  (bind _ballooncollision (Variables|Default|GetBalloonCollision))
  ;[1 1 1] Get Smacked
  (Networking|SwitchHasAuthority
    (:Authority
      (if (Variables|Default|GetbActive)
        (Utilities|Casting|CastToB_Melee_Impact_Generic DamageCauser)
        (bind _assphere_collision (Utilities|Casting|CastToSphereComponent FHitComponent))
        (if (Utilities|Array|Get(acopy) (Variables|Default|GetLivingBalloons) (Utilities|Array|FindItem _ballooncollision _assphere_collision))
          (Math|Integer|IncrementInt (Variables|Default|GetBurstBalloon)
            (:%20%20
              (bind _asfort_player_controller_athena (Utilities|Casting|CastToFortPlayerControllerAthena InstigatedBy))
              (CallFunction|ImpactClear (Utilities|Array|FindItem _ballooncollision _assphere_collision))
              (Class|BalloonBoardInterface|BurstBalloon(Message) (Utilities|Array|Get(acopy) (Variables|Default|GetBalloonRender) (Utilities|Array|FindItem _ballooncollision _assphere_collision)))
              (Utilities|Array|SetArrayElem (Variables|Default|GetLivingBalloons) (Utilities|Array|FindItem _ballooncollision _assphere_collision))
              (Class|ScoreboardInterface|MoveScore (Variables|Interaction|GetMyScoreBoard) 1 (Utilities|Array|MakeArray _asfort_player_controller_athena)))))))))

(fn BlueprintCanInteract
  (params (InteractingPawn :type Fort%20Pawn%20Object%20Reference) (InteractionBeingAttempted :type EInteractionBeingAttempted%20Enum :default "FirstInteraction") (InteractionType :type TInteractionType%20Enum :default "IT_NoInteraction"))
  (return (not (Variables|Default|GetbActive))))

(fn BlueprintGetInteractionString
  (params (InteractingPawn :type Fort%20Pawn%20Object%20Reference) (InteractionBeingAttempted :type EInteractionBeingAttempted%20Enum :default "FirstInteraction"))
  (return (Variables|Default|GetInteractionString)))

(fn OnRep_bActive
  (if (Variables|Default|GetbActive)
    (Variables|Default|SetStartingGap (Variables|Default|GetSpawnGap))
    (CallFunction|GameStartup)
    (CallFunction|RollRandoms)
    (else
      (Class|BPScoreBoardPlace|EndGame (Variables|Interaction|GetMyScoreBoard))
      (Audio|Components|Audio|Stop (Variables|Default|GetGame_Music))
      (Audio|SpawnSoundAttached "/Figment_S07/Sounds/Fort_Gadgets/Balloons/Balloon_Quest_End_Cue.Balloon_Quest_End_Cue" (Variables|Components|GetBoardRoot))
      (CallFunction|ForceTimerReset)
      (for _array_element (Variables|Default|GetBalloonRender)
        (Class|BalloonBoardInterface|KillBalloons(Message) _array_element)))))

(fn OnRep_BalloonIndex)

(fn OnRep_bSafeSpawn)

(fn OnRep_BurstBalloon)

(fn OnRep_CurrentBalloon)

(fn OnRep_Funky
  (Utilities|Time|SetTimerbyFunctionName 0 "DebugLoop" 0.015 true))

(fn OnRep_LivingBalloons)

(fn ShouldDie
  (params (Damage :type Float%20%28single-precision%29) (EventInstigator :type Controller%20Object%20Reference) (DamageCauser :type Actor%20Object%20Reference) (bIsTest :type Boolean) (EffectContext :type Gameplay%20Effect%20Context%20Handle%20Structure%20%28by%20ref%29))
  (return false))

(fn ConstructionScript
  ;[0.04501 0.1922 1] Store collision points in array (matching indices across all balloons)
  (Variables|Default|SetBalloonCollision (Utilities|Array|MakeArray (Variables|Default|GetB_01) (Variables|Default|GetB_02) (Variables|Default|GetB_03) (Variables|Default|GetB_04) (Variables|Default|GetB_05) (Variables|Default|GetB_06) (Variables|Default|GetB_07) (Variables|Default|GetB_08) (Variables|Default|GetB_09) (Variables|Default|GetB_10)))
  ;[0.04501 0.1922 1] Go through list of balloons, convert them to explicit class and store in array.
  (for (_array_element _array_index) (Utilities|Array|MakeArray (Variables|Default|GetB_Rend_01) (Variables|Default|GetB_Rend_02) (Variables|Default|GetB_Rend_03) (Variables|Default|GetB_Rend_04) (Variables|Default|GetB_Rend_05) (Variables|Default|GetB_Rend_06) (Variables|Default|GetB_Rend_07) (Variables|Default|GetB_Rend_08) (Variables|Default|GetB_Rend_09) (Variables|Default|GetB_Rend_10))
    (bind _asbp_popping_balloon (Utilities|Casting|CastToBP_PoppingBalloon (Variables|/Script/Engine.ChildActorComponent|GetChildActor _array_element)))
    (Utilities|Array|SetArrayElem (Variables|Default|GetBalloonRender) _array_index _asbp_popping_balloon true)))
